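# Material function for the mobile gravel layer of the Apollo landscape, built from world-aligned Z-projected dirt textures.
# A LandscapeLayer input blends tinted and lighter color, which Mountain, Swamp and Farm terrain masks recolor before clamped height alpha darkens it.
# The landscape output uses constant specular, roughness and vertex normals.
# A static mesh output keeps sampled normal and spec on default and deferred paths, using constants on forward and mobile.
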MaterialFunction /Clyde/Athena/Apollo/Environments/Landscape/Materials/Biome/CommonMaterials/Mobile/MF_Mobile_Common_Gravel
graph {nodes [
  n1 = Constant() {r=0.3}
  n2 = Constant() {r=0.9}
  n3 = MaterialFunctionCall!(UseLandscapeCoords=FunctionInput(Preview=StaticBool() {value=true}) {bUsePreviewValueAsDefault=true, inputName=UseLandscapeCoords, inputType=FunctionInput_StaticBool}) {materialFunction=/Clyde/Athena/Apollo/Environments/Landscape/Materials/Biome/CommonMaterials/MF_ApolloTerrainMasks}
  n4 = StaticBool() {value=true}
  n5 = Constant3Vector() {constant=<rgba 0.8 0.5869 0.4049 0>}
  n6 = Constant3Vector() {constant=<rgba 0.3465 0.3481 0.35 1>}
  n7 = Constant3Vector() {constant=<rgba 0.45 0.3773 0.225 1>}
  n8 = MaterialFunctionCall!(
   Diffuse Texture Object=TextureObject() {texture=/Game/Environments/Apollo/Sets/Terrain/Textures/T_Apollo_Dirt_D_01}
   Spec Texture Object=TextureObject() {samplerType=SAMPLERTYPE_Masks, texture=/Game/Environments/Apollo/Sets/Terrain/Textures/T_Apollo_Dirt_S_01}
   TextureSize=Constant() {r=550}
   --------------------------
   Use seperate settings for Z projection
   Use seperate texture for Z projection
   Texture Size Z projection
   Z Plane Diffuse Texture Object
   Z Plane Spec Texture Object
   "--------------------------      "
   ProjectionTransitionContrast
   Use Texture Alpha Heightmap to Refine the Transistion=StaticBool()
   HeightLerpTransitionContrast
   PickTheDiffuseHeightChannel
   TheDiffuseHeightChannel
   "--------------------------       "
   NormalTextureObject=TextureObject() {samplerType=SAMPLERTYPE_Normal, texture=/Game/Environments/Apollo/Sets/Terrain/Textures/T_Apollo_Dirt_N_01}
   World Space Normals=StaticBool()
   "--------------------------  "
   Export Float 4 Diffuse=StaticBool() {value=true}
   Export Float 4 Spec
   "--------------------------     "
   WorldPosition=WorldPosition() {worldPositionShaderOffset=WPT_ExcludeAllShaderOffsets}.XYZ
   World Space Normal
  ) {materialFunction=/Engine/Functions/Engine_MaterialFunctions02/Texturing/WorldAlignedTextures_Complex}
  n9 = ComponentMask(None=@n8.Z Diffuse Texture) {b=true, g=true, r=true}
  n10 = Multiply(A=Constant3Vector() {constant=<rgba 0.9 1 1 0>}, B=@n9)
  n11 = Constant() {r=1}
  n12 = MaterialFunctionCall!(
   UseBiomeColorMask=StaticBool() {value=true}
   BiomeColorMask=@n3.Mountain
   BiomeColorDiv=Constant3Vector() {constant=<rgba 1 0.7337 0.5061 0>}
   BiomeColor=Constant3Vector() {constant=<rgba 0.4275 0.4425 0.45 1>}
   Basecolor=LinearInterpolate(
    A=Multiply(A=@n7, B=@n10)
    B=@n10
    Alpha=FunctionInput(Preview=@n11) {bUsePreviewValueAsDefault=true, inputName=LandscapeLayer, inputType=FunctionInput_Scalar}
   )
  ) {materialFunction=/Clyde/Athena/Apollo/Environments/Landscape/Materials/Biome/CommonMaterials/MF_LS_DivideColor}
  n13 = ComponentMask(None=@n8.Z Diffuse Texture) {a=true}
  n14 = Multiply(
   A=MaterialFunctionCall(
    UseBiomeColorMask=StaticBool() {value=true}
    BiomeColorMask=@n3.Farm
    BiomeColorDiv=Constant3Vector() {constant=<rgba 0.8 0.7395 0.64 0>}
    BiomeColor=Constant3Vector() {constant=<rgba 0.9 0.7425 0.27 1>}
    Basecolor=MaterialFunctionCall(UseBiomeColorMask=@n4, BiomeColorMask=@n3.Swamp, BiomeColorDiv=@n5, BiomeColor=@n6, Basecolor=@n12.Result) {materialFunction=&MF_LS_DivideColor}.Result
   ) {materialFunction=&MF_LS_DivideColor}.Result
   B=Max(A=Constant() {r=0.85}, B=@n13)
  )
  n15 = SetMaterialAttributes!(MaterialAttributes, BaseColor=@n14, Specular=@n1, Roughness=@n2, Normal=VertexNormalWS()) {attributeSetTypes [
    #1
    #2
    #3
    #4
   ]}
  FunctionOutput(None=@n15) {desc=Mobile Landscape Only, outputName=Material, sortPriority=-1}
  FunctionOutput(None=AppendVector(A=@n10, B=FunctionOutput(None=@n13) {outputName=Height})) {outputName=RGBA, sortPriority=6}
  FunctionOutput(None=@n9) {outputName=RGB, sortPriority=6}
  n16 = MaterialFunctionCall(Float3=@n8.Z Spec) {materialFunction=&BreakOutFloat3Components}
  n17 = SetMaterialAttributes(BaseColor=@n14, Specular=@n16.R, Roughness=@n16.R, Normal=@n8.Normal Z Texture) {attributeSetTypes [
    #1
    #2
    #3
    #4
   ]}
  n18 = SetMaterialAttributes(BaseColor=@n14, Specular=@n1, Roughness=@n2, Normal=Constant3Vector() {constant=<rgba 0 0 1 0>}) {attributeSetTypes [
    #1
    #2
    #3
    #4
   ]}
  FunctionOutput(None=ShadingPathSwitch(Default=@n17, Deferred=@n17, Forward=@n18, Mobile=@n18)) {outputName=Output_for_StaticMeshs, sortPriority=80}
  FunctionOutput(None=@n14) {outputName=Mobile-RGB, sortPriority=6}
  Multiply(
   A=VectorParameter() {group=Dirt, parameterName=DirtColor, defaultValue=<rgba 0.45 0.3773 0.225 1>}.RGB
   B=Multiply(A=VectorParameter() {group=Dirt, parameterName=DirtColor1, defaultValue=<rgba 0.95 1 1 1>}.RGB, B=@n9)
  )
  VectorParameter() {group=Dirt, parameterName=DirtColor-dark1, defaultValue=<rgba 0.45 0.3773 0.225 1>}
 ]}
