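# Material function that builds the shared Apollo terrain cliff surface, by its name for static meshes, outputting material attributes plus a height from diffuse alpha.
# WorldAlignedTextures_Complex projects cliff diffuse, spec and normal maps in world space at a TilingSize input, using only its XY outputs.
# A scalar ColorTint input multiplies base color; the ColorTint_RockCliff vector parameter is unconnected.
# A shading path switch swaps spec-map specular and roughness for constants on forward and mobile.
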MaterialFunction /Clyde/Athena/Apollo/Environments/Terrain_Meshes/MaterialFunctions/MF_Common_Clff_StaticMesh
graph {nodes [
  n1 = MaterialFunctionCall!(
   Diffuse Texture Object=TextureObject() {texture=/Game/Environments/Apollo/Sets/Terrain/Textures/T_Apollo_Cliff_D_01}
   Spec Texture Object=TextureObject() {samplerType=SAMPLERTYPE_Masks, texture=/Game/Environments/Apollo/Sets/Terrain/Textures/T_Apollo_Cliff_S_01}
   TextureSize=FunctionInput(Preview=Constant() {r=1000}) {bUsePreviewValueAsDefault=true, inputName=TilingSize, inputType=FunctionInput_Scalar}
   --------------------------
   Use seperate settings for Z projection
   Use seperate texture for Z projection
   Texture Size Z projection
   Z Plane Diffuse Texture Object
   Z Plane Spec Texture Object
   "--------------------------      "
   ProjectionTransitionContrast
   Use Texture Alpha Heightmap to Refine the Transistion=StaticBool()
   HeightLerpTransitionContrast
   PickTheDiffuseHeightChannel
   TheDiffuseHeightChannel
   "--------------------------       "
   NormalTextureObject=TextureObject() {samplerType=SAMPLERTYPE_Normal, texture=/Game/Environments/Apollo/Sets/Terrain/Textures/T_Apollo_Cliff_N_01}
   World Space Normals=StaticBool() {value=true}
   "--------------------------  "
   Export Float 4 Diffuse=StaticBool() {value=true}
   Export Float 4 Spec
   "--------------------------     "
   WorldPosition=WorldPosition() {worldPositionShaderOffset=WPT_ExcludeAllShaderOffsets}.XYZ
   World Space Normal
  ) {materialFunction=/Engine/Functions/Engine_MaterialFunctions02/Texturing/WorldAlignedTextures_Complex}
  n2 = Multiply(
   A=FunctionInput(Preview=Constant() {r=1}) {bUsePreviewValueAsDefault=true, inputName=ColorTint, inputType=FunctionInput_Scalar}
   B=ComponentMask(None=@n1.XY Diffuse Texture) {b=true, g=true, r=true}
  )
  n3 = SetMaterialAttributes!(
   MaterialAttributes
   BaseColor=@n2
   Specular=ComponentMask(None=@n1.XY Spec) {r=true}
   Roughness=ComponentMask(None=@n1.XY Spec) {b=true}
   Normal=@n1.Normal XY Texture
  ) {attributeSetTypes [
    #1
    #2
    #3
    #4
   ]}
  n4 = SetMaterialAttributes(BaseColor=@n2, Specular=Constant() {r=0.3}, Roughness=Constant() {r=0.9}, Normal=@n1.Normal XY Texture) {attributeSetTypes [
    #1
    #2
    #3
    #4
   ]}
  FunctionOutput(None=ShadingPathSwitch(Default=@n3, Deferred=@n3, Forward=@n4, Mobile=@n4))
  FunctionOutput(None=ComponentMask(None=@n1.XY Diffuse Texture) {a=true}) {outputName=Height, sortPriority=3}
  VectorParameter() {parameterName=ColorTint_RockCliff, defaultValue=<rgba 0.8698 0.8589 0.6537 1>}
 ]}
