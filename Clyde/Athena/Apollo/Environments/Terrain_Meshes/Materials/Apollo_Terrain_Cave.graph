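# Cave terrain material for Apollo environment meshes, blending gravel, cliff rock and ceiling rock layers with a dirt physical material.
# Vertex green picks cave or cliff tint, passed through height-based darkening into both rock functions.
# Downward-facing normals drive the height-blended ceiling transition; vertex alpha height-blends gravel against that rock before MF_CreativeMoveTool.
# Normals are world-space; forward and mobile paths replace the gravel normal with the vertex normal.
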Material /Clyde/Athena/Apollo/Environments/Terrain_Meshes/Materials/Apollo_Terrain_Cave
graph {
 properties {
  bTangentSpaceNormal=false
  bUseMaterialAttributes=true
  bUsedWithInstancedStaticMeshes=true
  bUsedWithSplineMeshes=true
  ditheredLODTransition=true
  physMaterial=/Game/FortressPhysicalMaterials/Dirt
  pixelDepthOffsetMode=PDOM_Legacy
  usedShadingModels=Default Lit
 }
 nodes [
  LinearInterpolate(
   A=VectorParameter() {parameterName=ColorTint_RockCave, defaultValue=<rgba 0.7083 0.4385 0.1586 1>}.RGB
   B=VectorParameter() {parameterName=ColorTint_RockCliff, defaultValue=<rgba 0.8698 0.8589 0.6537 1>}.RGB
   Alpha=VertexColor().G
  )
  n1 = MaterialFunctionCall!(DarknessAmount=Constant() {r=0.2}) {materialFunction=/Clyde/Athena/Apollo/Environments/Landscape/Materials/Biome/CommonMaterials/MF_HeightBasedDarkenCliff}
  n2 = MaterialFunctionCall!(
   ColorTint=@n1.Result
   TilingSize=ScalarParameter() {defaultValue=582.8452, parameterName=SizeRockCeiling}
   Use_Shift_Worldposition=StaticBool()
  ) {materialFunction=/Clyde/Athena/Apollo/Environments/Terrain_Meshes/MaterialFunctions/MF_EdgeRocks}
  n3 = Multiply(A=ComponentMask(None=VertexNormalWS()) {b=true}) {constB=-1}
  n4 = MaterialFunctionCall(
   Transition Phase=Power(Base=Saturate(None=@n3), Exp=ScalarParameter() {defaultValue=1.7534, parameterName=CeilingTransition})
   Height Texture=@n2.Z Ceiling Height
   Contrast=ScalarParameter() {parameterName=ContrastCeilingTransition}
  ) {materialFunction=&HeightLerp}
  n5 = MaterialFunctionCall!(ColorTint=@n1.Result, TilingSize=ScalarParameter() {defaultValue=669.0363, parameterName=SizeRockCliffs}) {materialFunction=/Clyde/Athena/Apollo/Environments/Terrain_Meshes/MaterialFunctions/MF_Common_Clff_StaticMesh}
  n6 = VertexColor()
  n7 = ScalarParameter() {parameterName=ContrastFloorTransition}
  n8 = GetMaterialAttributes(None=MaterialFunctionCall!(UseLandscapeCoords, LandscapeLayer) {materialFunction=/Clyde/Athena/Apollo/Environments/Landscape/Materials/Biome/CommonMaterials/Mobile/MF_Mobile_Common_Gravel}.Output_for_StaticMeshs) {attributeGetTypes [#1]}
  n9 = Transform(None=@n8.Normal)
  n10 = VertexNormalWS()
  n11 = ShadingPathSwitch(Default=@n9, Deferred=@n9, Forward=@n10, Mobile=@n10)
  MaterialFunctionCall(MA In=BlendMaterialAttributes(
    A=SetMaterialAttributes!(MaterialAttributes=@n8.MaterialAttributes, Normal=@n11) {attributeSetTypes [#1]}
    B=BlendMaterialAttributes(A=@n5.Result, B=@n2."Z Ceiling ", Alpha=@n4.Alpha)
    Alpha=MaterialFunctionCall(Transition Phase=@n6.A, Height Texture=@n5.Height, Contrast=@n7) {materialFunction=&HeightLerp}.Alpha
   )) {materialFunction=&MF_CreativeMoveTool}
  StaticBool() {value=true}
 ]
}
