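# Material function for Apollo terrain meshes that builds a parameterized custom material layer, outputting material attributes plus a Height value from Diffuse alpha.
# Tinted Diffuse feeds base color, while the Specular mask's red scales specular and blue interpolates between low and high roughness.
# A ShadingPathSwitch gives default and deferred paths these maps plus a transformed normal map.
# Forward and mobile keep the tinted color but use constant specular, roughness and flat normal.
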MaterialFunction /Clyde/Athena/Apollo/Environments/Terrain_Meshes/MaterialFunctions/MF_Common_GenericMaterial
graph {nodes [
  n1 = TextureSampleParameter2D() {
   group=BlendedCustomMaterial
   parameterName=Diffuse
   samplerSource=SSM_Wrap_WorldGroupSettings
   texture=/Game/Characters/Fortress_Character/Textures/T_All_White
  }
  n2 = Multiply(A=VectorParameter() {group=BlendedCustomMaterial, parameterName=ColorTint_CustomMaterial, defaultValue=<rgba 1 1 1 1>}.RGB, B=@n1.RGB)
  n3 = SetMaterialAttributes!(
   MaterialAttributes
   BaseColor=@n2
   Specular=Constant() {r=0.3}
   Roughness=Constant() {r=0.9}
   Normal=Constant3Vector() {constant=<rgba 0 0 1 0>}
  ) {attributeSetTypes [
    #1
    #2
    #3
    #4
   ]}
  n4 = TextureSampleParameter2D() {
   group=BlendedCustomMaterial
   parameterName=Specular
   samplerSource=SSM_Wrap_WorldGroupSettings
   samplerType=SAMPLERTYPE_Masks
   texture=&T_Fortnite_Default_S
  }
  n5 = TextureSampleParameter2D() {
   group=BlendedCustomMaterial
   parameterName=Normal
   samplerSource=SSM_Wrap_WorldGroupSettings
   samplerType=SAMPLERTYPE_Normal
   texture=&FlatNormalBlue
  }
  n6 = SetMaterialAttributes(
   BaseColor=@n2
   Specular=Multiply(A=ScalarParameter() {defaultValue=1, group=BlendedCustomMaterial, parameterName=Specular_BlendCustomMaterial}, B=@n4.R)
   Roughness=LinearInterpolate(
    A=ScalarParameter() {group=BlendedCustomMaterial, parameterName=Roughness_BlendCustomMaterial_low}
    B=ScalarParameter() {defaultValue=1, group=BlendedCustomMaterial, parameterName=Roughness_BlendCustomMaterial_High}
    Alpha=@n4.B
   )
   Normal=Transform(None=@n5.RGB)
  ) {attributeSetTypes [
    #1
    #2
    #3
    #4
   ]}
  FunctionOutput(None=ShadingPathSwitch(Default=@n6, Deferred=@n6, Forward=@n3, Mobile=@n3))
  FunctionOutput(None=@n1.A) {outputName=Height, sortPriority=3}
 ]}
