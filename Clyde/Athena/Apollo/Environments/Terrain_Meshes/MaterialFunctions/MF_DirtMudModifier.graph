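# Material function that adds an optional, tinted mud layer with puddles to incoming terrain material attributes.
# The inverted heightmap input, offset and scaled by puddle height and blend parameters, forms a saturated mask for low areas.
# That mask interpolates specular and roughness multipliers and flattens the normal, while a tint darkens base color throughout the mud.
# A static switch enables the effect, blended over the original attributes by vertex color alpha.
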MaterialFunction /Clyde/Athena/Apollo/Environments/Terrain_Meshes/MaterialFunctions/MF_DirtMudModifier
graph {nodes [
  n1 = FunctionInput() {inputType=FunctionInput_MaterialAttributes}
  n2 = GetMaterialAttributes(None=@n1) {attributeGetTypes [
    #1
    #2
    #3
    #4
   ]}
  n3 = ScalarParameter() {defaultValue=1, group=Mud, parameterName=Specular_MudHigh}
  n4 = ScalarParameter() {defaultValue=1, group=Mud, parameterName=Specular_MudLow}
  n5 = OneMinus(None=FunctionInput() {inputName=Heightmap, inputType=FunctionInput_Scalar, sortPriority=3})
  n6 = ScalarParameter() {defaultValue=-0.5462, group=Mud, parameterName=MudPuddleHeight}
  n7 = Saturate(None=Divide(A=Add(A=@n5, B=@n6), B=ScalarParameter() {defaultValue=0.1333, group=Mud, parameterName=MudPuddleBlend}))
  n8 = ScalarParameter() {defaultValue=1.266, group=Mud, parameterName=Roughness_MudHigh}
  n9 = ScalarParameter() {defaultValue=-0.1061, group=Mud, parameterName=Roughness_MudLow}
  n10 = SetMaterialAttributes!(
   MaterialAttributes=@n2.MaterialAttributes
   BaseColor=Multiply(A=@n2.BaseColor, B=VectorParameter() {group=Mud, parameterName=MudTint, defaultValue=<rgba 0.1198 0.1129 0.1098 1>}.RGB)
   Specular=Multiply(A=@n2.Specular, B=LinearInterpolate(A=@n3, B=@n4, Alpha=@n7))
   Roughness=Multiply(A=@n2.Roughness, B=LinearInterpolate(A=@n8, B=@n9, Alpha=@n7))
   Normal=LinearInterpolate(A=@n2.Normal, B=Constant3Vector() {constant=<rgba 0 0 1 0>}, Alpha=@n7)
  ) {attributeSetTypes [
    #1
    #2
    #3
    #4
   ]}
  n11 = VertexColor()
  FunctionOutput(None=StaticSwitchParameter(True=BlendMaterialAttributes(A=@n1, B=@n10, Alpha=@n11.A), False=@n1) {parameterName=UseMud-UsesAlphaVertColor})
 ]}
