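# Material function for terrain meshes that blends grass, dirt, sand, asphalt and cliff material-attribute layers using vertex color masks.
# Static switches toggle layers, composited through height-lerps: sand into dirt, dirt with grass, then asphalt, then cliffs.
# Transition masks also tint and darken base color at layer edges.
# Forward and mobile paths skip asphalt's edge normal blend and substitute vertex normals for the world-transformed normal before cliff blending.
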MaterialFunction /Clyde/Athena/Apollo/Environments/Terrain_Meshes/MaterialFunctions/MF_VextexColorLayerBlend
graph {nodes [
  FunctionInput() {inputName=Sand_Height, inputType=FunctionInput_Scalar, sortPriority=111}
  n1 = GetMaterialAttributes(None=FunctionInput() {inputName=Cliffs, inputType=FunctionInput_MaterialAttributes, sortPriority=12}) {attributeGetTypes [#1]}
  n2 = ScalarParameter() {defaultValue=0.01, parameterName=TransitionGrassContrast}
  n3 = ScalarParameter() {defaultValue=2.3735, parameterName=TransitionGrassContrast2}
  n4 = FunctionInput() {inputName=Cliffs_Height, inputType=FunctionInput_Scalar, sortPriority=13}
  n5 = FunctionInput() {inputName=Grass_Height, inputType=FunctionInput_Scalar, sortPriority=2}
  n6 = MaterialFunctionCall(
   Transition Phase=Power(Base=VertexColor() {desc=mask cliff to grass transition}.A, Exp=LinearInterpolate(A=@n2, B=@n3, Alpha=@n4))
   Height Texture=ComponentMask(None=OneMinus(None=@n5)) {r=true}
   Contrast=ScalarParameter() {group=Cliffs, parameterName=CliffLerpContrast}
  ) {materialFunction=&HeightLerp}
  n7 = SetMaterialAttributes!(MaterialAttributes=@n1.MaterialAttributes, BaseColor=Multiply(A=@n1.BaseColor, B=@n6.Lerp Alpha No Contrast)) {attributeSetTypes [#1]}
  n8 = FunctionInput() {inputName=Asphalt, inputType=FunctionInput_MaterialAttributes, sortPriority=8}
  n9 = GetMaterialAttributes(None=@n8) {attributeGetTypes [#2]}
  n10 = FunctionInput() {inputName=Asphalt_Normal, sortPriority=9}
  n11 = Constant3Vector() {constant=<rgba 0 0 1 0>}
  n12 = VertexColor()
  n13 = FunctionInput() {inputName=Asphalt_Height, inputType=FunctionInput_Scalar, sortPriority=9}
  n14 = Saturate(None=LinearInterpolate(
    A=ScalarParameter() {defaultValue=-0.33, parameterName=ContrastAsphaltBlend1}
    B=ScalarParameter() {defaultValue=1.2741, parameterName=ContrastAsphaltBlend2}
    Alpha=MaterialFunctionCall(Transition Phase=@n12.R, Height Texture=@n13) {materialFunction=&HeightLerp}.Lerp Alpha No Contrast
   ))
  n15 = SetMaterialAttributes!(
   MaterialAttributes=@n9.MaterialAttributes
   Normal=MaterialFunctionCall(BaseNormal=LinearInterpolate(A=@n10, B=@n11, Alpha=@n14), AdditionalNormal=@n9.Normal) {materialFunction=&BlendAngleCorrectedNormals}.Result
  ) {attributeSetTypes [#2]}
  n16 = ShadingPathSwitch(Default=@n15, Deferred=@n15, Forward=@n8, Mobile=@n8)
  n17 = FunctionInput() {inputName=Grass, inputType=FunctionInput_MaterialAttributes}
  n18 = GetMaterialAttributes(None=@n17) {attributeGetTypes [#1]}
  n19 = Constant3Vector() {constant=<rgba 2 1.0029 0.6 1>}
  n20 = VertexColor() {desc=mask cliff to grass transition}
  n21 = Add(A=@n20.R, B=@n20.R)
  n22 = OneMinus(None=StaticSwitchParameter(True=Saturate(None=@n21), False=@n20.B) {defaultValue=true, group=Layers, parameterName=UseConcrete})
  n23 = FunctionInput() {inputName=Dirt_Height, inputType=FunctionInput_Scalar, sortPriority=6}
  n24 = MaterialFunctionCall(Transition Phase=@n22, Height Texture=@n23) {materialFunction=&HeightLerp}
  n25 = Add(A=Constant() {r=-1}, B=Multiply(A=Constant() {r=2}, B=@n24.Lerp Alpha No Contrast))
  n26 = SetMaterialAttributes!(
   MaterialAttributes=@n18.MaterialAttributes
   BaseColor=LinearInterpolate(A=Multiply(A=@n18.BaseColor, B=@n19), B=@n18.BaseColor, Alpha=Saturate(None=@n25))
  ) {attributeSetTypes [#1]}
  n27 = VertexColor()
  n28 = ScalarParameter() {parameterName=SandBlendContrast}
  n29 = FunctionInput() {inputName=Dirt, inputType=FunctionInput_MaterialAttributes, sortPriority=5}
  n30 = StaticSwitchParameter(
   True=BlendMaterialAttributes(
    A=@n29
    B=FunctionInput() {inputName=Sand, inputType=FunctionInput_MaterialAttributes, sortPriority=99}
    Alpha=MaterialFunctionCall(Transition Phase=@n27.G, Height Texture=@n23, Contrast=@n28) {materialFunction=&HeightLerp}.Alpha
   )
   False=@n29
  ) {group=Layers, parameterName=UseSand}
  n31 = GetMaterialAttributes(None=@n30) {attributeGetTypes [#1]}
  n32 = Multiply(A=@n31.BaseColor, B=Constant3Vector() {constant=<rgba 0.65 0.608 0.52 1>})
  n33 = StaticSwitchParameter(
   True=SetMaterialAttributes(MaterialAttributes=@n31.MaterialAttributes, BaseColor=LinearInterpolate(A=@n31.BaseColor, B=@n32, Alpha=@n24.Lerp Alpha No Contrast)) {attributeSetTypes [#1]}
   False=@n30
  ) {defaultValue=true, group=Layers, parameterName=UseGrass}
  n34 = GetMaterialAttributes(None=@n33) {attributeGetTypes [#1]}
  n35 = Multiply(A=@n34.BaseColor, B=VectorParameter() {parameterName=AsphaltEdgeShadow, defaultValue=<rgba 0.2656 0.2639 0.2629 1>}.RGB)
  n36 = VertexColor()
  n37 = ScalarParameter() {defaultValue=21.251, parameterName=ContrastAsphaltDirtShift}
  n38 = MaterialFunctionCall(Transition Phase=Subtract(A=@n12.R, B=Power(Base=@n36.B, Exp=@n37)), Height Texture=@n23) {materialFunction=&HeightLerp}
  n39 = Saturate(None=LinearInterpolate(
    A=ScalarParameter() {defaultValue=-0.5508, parameterName=ContrastAsphaltDirtBlend1}
    B=ScalarParameter() {defaultValue=1.9018, parameterName=ContrastAsphaltDirtBlend2}
    Alpha=@n38.Lerp Alpha No Contrast
   ))
  n40 = StaticSwitchParameter(
   True=SetMaterialAttributes(MaterialAttributes=@n34.MaterialAttributes, BaseColor=LinearInterpolate(A=@n34.BaseColor, B=@n35, Alpha=@n39)) {attributeSetTypes [#1]}
   False=@n33
  ) {defaultValue=true, group=Layers, parameterName=UseConcrete}
  n41 = ScalarParameter() {defaultValue=1.9677, parameterName=T_Dirt}
  n42 = ScalarParameter() {defaultValue=1.9703, parameterName=T_Grass}
  n43 = Saturate(None=MaterialFunctionCall!(
    Diffuse 1
    Diffuse 2
    Transistion Phase=@n22
    Height Texture 1=Multiply(A=@n23, B=@n41)
    Height Texture 2=Multiply(A=@n5, B=@n42)
    Contrast=ScalarParameter() {parameterName=Dirt_Grass_transitioncontrast}
   ) {materialFunction=/Engine/Functions/Engine_MaterialFunctions02/Texturing/HeightLerpWithTwoHeightMaps}.Alpha)
  n44 = StaticSwitchParameter(
   True=StaticSwitchParameter(True=BlendMaterialAttributes(A=@n26, B=@n40, Alpha=@n43), False=@n40) {defaultValue=true, group=Layers, parameterName=UseGrass}
   False=@n17
  ) {defaultValue=true, group=Layers, parameterName=UseDirt}
  n45 = GetMaterialAttributes(None=StaticSwitchParameter(True=BlendMaterialAttributes(A=@n44, B=@n16, Alpha=@n14), False=@n44) {defaultValue=true, group=Layers, parameterName=UseConcrete}) {attributeGetTypes [#2]}
  n46 = VertexNormalWS()
  n47 = Transform(None=@n45.Normal)
  n48 = SetMaterialAttributes!(MaterialAttributes=@n45.MaterialAttributes, Normal=ShadingPathSwitch(Default=@n47, Deferred=@n47, Forward=@n46, Mobile=@n46)) {attributeSetTypes [#2]}
  FunctionOutput(None=StaticSwitchParameter(True=BlendMaterialAttributes(A=@n48, B=@n7, Alpha=@n6.Alpha), False=@n48) {desc=Diffuse, group=Layers, parameterName="UseCliffs (Disables Mud)"})
  n49 = ScalarParameter() {defaultValue=-40, parameterName=DirtNormalEdge}
  LinearInterpolate(B=Constant3Vector() {constant=<rgba 0 0 1 0>}, Alpha=Saturate(None=Multiply(A=@n49, B=@n38.Lerp Alpha No Contrast)))
 ]}
