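# Master material for Apollo biome terrain meshes that blends them into the surrounding landscape.
# Vertex colors drive a height-aware blend of grass, dirt, asphalt, cliff or custom material, and sand, with optional edge rocks and masked paint layer.
# Landscape runtime virtual textures supply height-distance blending into terrain color and normals, with slope-based mip bias.
# Low and Medium quality or non-virtual-texture paths skip edge rocks and terrain blending, using tinted dirt and sand.
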Material /Clyde/Athena/Apollo/Environments/Terrain_Meshes/Materials/Apollo_Terrain_Mat_Biome
graph {
 properties {
  bTangentSpaceNormal=false
  bUseMaterialAttributes=true
  bUsedWithInstancedStaticMeshes=true
  bUsedWithNanite=true
  bUsedWithSplineMeshes=true
  physMaterial=/Game/FortressPhysicalMaterials/Grass
  pixelDepthOffsetMode=PDOM_Legacy
  refractionMethod=RM_IndexOfRefraction
  usedShadingModels=Default Lit
 }
 nodes [
  n1 = LinearInterpolate(
   A=LinearInterpolate(
    A=Constant3Vector() {constant=<rgba 0.0073 0.045 0.0027 1>}
    B=Constant3Vector() {constant=<rgba 0.135 0.0934 0.0614 1>}
    Alpha=VertexColor().B
   )
   B=Constant3Vector() {constant=<rgba 0.03 0.0192 0.0099 1>}
   Alpha=VertexColor().A
  )
  n2 = Constant3Vector() {constant=<rgba 0.18 0.1795 0.1791 1>}
  n3 = VertexColor()
  Reroute(None=MakeMaterialAttributes!(
    BaseColor=LinearInterpolate(A=@n1, B=@n2, Alpha=@n3.R)
    Metallic=Constant()
    Specular
    Roughness=Constant() {r=1}
    Anisotropy
    EmissiveColor
    Opacity
    OpacityMask
    Normal=VertexNormalWS()
    Tangent
    WorldPositionOffset
    SubsurfaceColor
    ClearCoat
    ClearCoatRoughness
    AmbientOcclusion
    Refraction
    CustomizedUVs_0
    CustomizedUVs_1
    CustomizedUVs_2
    CustomizedUVs_3
    CustomizedUVs_4
    CustomizedUVs_5
    CustomizedUVs_6
    CustomizedUVs_7
    PixelDepthOffset
    ShadingModel
    Displacement
   ))
  ScalarParameter() {group=PaintLayer, parameterName=Paint Metallic}
  n4 = DotProduct(A=VertexNormalWS(), B=Constant3Vector() {constant=<rgba 0 0 1 0>})
  n5 = RuntimeVirtualTextureSample!(
   UVs
   World Position
   Mip Bias=LinearInterpolate(
    A=ScalarParameter() {defaultValue=6, parameterName=MipBiasVerticalSurfaces}
    B=ScalarParameter() {parameterName=MipBiasHorizontalSurfaces}
    Alpha=Saturate(None=@n4)
   )
  ) {materialType=BaseColor_Normal_Specular_YCoCg, mipValueMode=RVTMVM_MipBias, virtualTexture=/Game/Athena/Apollo/RVT/RVT_Apollo_Landscape}
  n6 = SetMaterialAttributes!(MaterialAttributes, Base Color=@n5.BaseColor, Specular=@n5.BaseColor, Roughness=@n5.BaseColor, Normal=@n5.BaseColor) {attributeSetTypes [
    #1
    #2
    #3
    #4
   ]}
  n7 = RuntimeVirtualTextureSample!(UVs, World Position) {materialType=WorldHeight, virtualTexture=/Game/Athena/Apollo/RVT/RVT_Apollo_Landscape_Height}
  n8 = ComponentMask(None=WorldPosition().XYZ) {b=true}
  n9 = Saturate(None=Subtract(A=VertexInterpolator(VS=@n7.WorldHeight).PS, B=@n8))
  n10 = Distance(A=@n7.WorldHeight, B=@n8)
  n11 = ScalarParameter() {defaultValue=60, parameterName=BlendwithTerrainDistance}
  n12 = Multiply(A=OneMinus(None=Ceil(None=@n9)), B=Saturate(None=Divide(A=@n10, B=@n11)))
  n13 = StaticSwitchParameter(True=Add(A=@n12, B=VertexColor().G), False=@n12) {defaultValue=true, parameterName="Use Vertex Color(G) for RVT Blending Mask (May break Sand Blending)"}
  n14 = TextureCoordinate()
  n15 = VectorParameter() {group=PaintLayer, parameterName=PaintLayerUV_RG_Offset_BA_Scale, defaultValue=<rgba 0 0 1 1>}
  n16 = ComponentMask(None=@n15.RGB) {g=true, r=true}
  n17 = TextureSampleParameter2D(UVs=Multiply(A=Add(A=@n14, B=@n16), B=AppendVector(A=@n15.B, B=@n15.B))) {
   group=PaintLayer
   parameterName=PaintLayerMasks
   samplerType=SAMPLERTYPE_Masks
   texture=/Game/Packages/DS_Fortnight/SM/Textures/T_SoccerTennisBasketball
  }
  n18 = StaticComponentMaskParameter(None=AppendVector(A=@n17.RGB, B=@n17.RGB)) {defaultR=true, group=PaintLayer, parameterName=PaintLayerMaskChannel}
  n19 = WorldPosition()
  n20 = TextureSampleParameter2D(UVs=Divide(
    A=ComponentMask(None=@n19.XYZ) {g=true, r=true}
    B=ScalarParameter() {defaultValue=2048, group=PaintLayer, parameterName=PaintLayerMask_Mask Tiling}
   ) {constB=1024}) {
   group=PaintLayer
   parameterName=PaintLayerMasks_Mask
   samplerType=SAMPLERTYPE_Masks
   texture=/Game/Packages/Fortnite_UniqueMaterials_1/Textures/T_FRT_Conc_Masks_B_new
  }
  n21 = AppendVector(A=@n20.RGB, B=@n20.RGB)
  n22 = StaticSwitchParameter(
   True=Multiply(A=@n18, B=StaticComponentMaskParameter(None=@n21) {defaultR=true, group=PaintLayer, parameterName=PaintLayerMask_Mask_Channel})
   False=@n18
  ) {desc=Diffuse, group=PaintLayer, parameterName=Use Second PaintLayer Mask}
  n23 = Multiply(A=@n22, B=ScalarParameter() {defaultValue=1, group=PaintLayer, parameterName=PaintLayer Alpha Modulation})
  n24 = SetMaterialAttributes(
   Base Color=VectorParameter() {group=PaintLayer, parameterName=PaintLayerColor, defaultValue=<rgba 1 1 1 1>}.RGB
   Specular=Constant() {r=0.5}
   Roughness=ScalarParameter() {defaultValue=0.25, group=PaintLayer, parameterName=Paint Roughness}
   Normal=VertexNormalWS()
  ) {attributeSetTypes [
    #1
    #2
    #3
    #4
   ]}
  n25 = MaterialFunctionCall!(DarknessAmount=Constant() {r=0.2}) {materialFunction=/Clyde/Athena/Apollo/Environments/Landscape/Materials/Biome/CommonMaterials/MF_HeightBasedDarkenCliff}
  n26 = MaterialFunctionCall!(ColorTint=@n25.Result, TilingSize=ScalarParameter() {defaultValue=400, parameterName=EdgeRockTiling}, Use_Shift_Worldposition) {materialFunction=/Clyde/Athena/Apollo/Environments/Terrain_Meshes/MaterialFunctions/MF_EdgeRocks}
  n27 = ComponentMask(None=VertexNormalWS()) {b=true}
  n28 = RuntimeVirtualTextureSample() {materialType=WorldHeight, virtualTexture=&RVT_Apollo_Landscape_Height}
  n29 = ComponentMask(None=WorldPosition().XYZ) {b=true}
  n30 = Ceil(None=Saturate(None=Subtract(A=@n28.WorldHeight, B=@n29)))
  n31 = Distance(A=@n28.WorldHeight, B=@n29)
  n32 = ScalarParameter() {defaultValue=500, parameterName=BlendwithTerrainDistance_Rocks}
  n33 = Saturate(None=Multiply(A=OneMinus(None=@n30), B=Divide(A=@n31, B=@n32)))
  n34 = Multiply(
   A=OneMinus(None=Saturate(None=@n27))
   B=Multiply(A=OneMinus(None=@n33), B=ScalarParameter() {defaultValue=0.8, parameterName=EdgeBlend_Transition2})
  )
  n35 = MaterialFunctionCall!(Slope-CliffMask, CombinedLandscapeLayersMask) {materialFunction=/Clyde/Athena/Apollo/Environments/Landscape/Materials/Biome/CommonMaterials/Mobile/MF_Mobile_Common_Grass}
  n36 = MaterialFunctionCall!(UseLandscapeCoords=StaticBool(), LandscapeLayer) {materialFunction=/Clyde/Athena/Apollo/Environments/Landscape/Materials/Biome/CommonMaterials/Mobile/MF_Mobile_Common_Gravel}
  n37 = MaterialFunctionCall!(In=@n36.Output_for_StaticMeshs, Heightmap=@n36.Output_for_StaticMeshs) {materialFunction=/Clyde/Athena/Apollo/Environments/Terrain_Meshes/MaterialFunctions/MF_DirtMudModifier}
  n38 = GetMaterialAttributes(None=@n37.Result) {attributeGetTypes [#1]}
  n39 = SetMaterialAttributes!(
   MaterialAttributes=@n38.MaterialAttributes
   Base Color=Multiply(A=VectorParameter() {parameterName=Mobile-DirtColorTint_KL, defaultValue=<rgba 1 1 1 1>}.RGB, B=@n38.Base Color)
  ) {attributeSetTypes [#1]}
  n40 = MaterialFunctionCall!(TilingSize) {materialFunction=/Clyde/Athena/Apollo/Environments/Landscape/Materials/MaterialFunctions/MF_ApolloAsphalt}
  n41 = MaterialFunctionCall() {materialFunction=/Clyde/Athena/Apollo/Environments/Terrain_Meshes/MaterialFunctions/MF_Common_GenericMaterial}
  n42 = MaterialFunctionCall!(ColorTint=@n25.Result, TilingSize=ScalarParameter() {defaultValue=1000, parameterName=SizeRockCliffs}) {materialFunction=/Clyde/Athena/Apollo/Environments/Terrain_Meshes/MaterialFunctions/MF_Common_Clff_StaticMesh}
  n43 = MaterialFunctionCall() {materialFunction=/Clyde/Athena/Apollo/Environments/Landscape/Materials/Biome/CommonMaterials/Mobile/MF_Mobile_Common_Sand}
  n44 = GetMaterialAttributes(None=@n43.OutPut_Staticmesh) {attributeGetTypes [#1]}
  n45 = SetMaterialAttributes(
   MaterialAttributes=@n44.MaterialAttributes
   Base Color=Multiply(A=VectorParameter() {parameterName=Mobile-SandColorTint_KL, defaultValue=<rgba 1 1 1 1>}.RGB, B=@n44.Base Color)
  ) {attributeSetTypes [#1]}
  n46 = MaterialFunctionCall!(
   Grass=@n35.Output_for_StaticMesh
   Grass_Height=@n35.Output_for_StaticMesh
   Dirt=VirtualTextureFeatureSwitch!(
    No=@n39
    Yes=QualitySwitch(Default=@n37.Result, Low=@n39, Medium=@n39) {desc="Mobile was plugged into High/default - swapped around\r\nto make grass consistent/nblend on all platforms"}
   )
   Dirt_Height=@n36.Height
   Asphalt=@n40.Result
   Asphalt_Height=@n40.Result
   Asphalt_Normal=@n40.Result
   Cliffs=StaticSwitchParameter(True=@n41.Result, False=@n42.Result) {parameterName=UseCustomMaterialBlend}
   Cliffs_Height=StaticSwitchParameter(True=@n41.Height, False=@n42.Height) {parameterName=UseCustomMaterialBlend}
   Sand=VirtualTextureFeatureSwitch(
    No=@n45
    Yes=QualitySwitch(Default=@n43.OutPut_Staticmesh, Low=@n45, Medium=@n45) {desc="Mobile was plugged into High/default - swapped around\r\nto make grass consistent/nblend on all platforms"}
   )
   Sand_Height=@n43.Height
  ) {materialFunction=/Clyde/Athena/Apollo/Environments/Terrain_Meshes/MaterialFunctions/MF_VextexColorLayerBlend}
  n47 = BlendMaterialAttributes(
   A=@n46.Result
   B=@n26.XY Walls
   Alpha=MaterialFunctionCall(
    Transition Phase=VertexInterpolator(VS=@n34).PS
    Height Texture=@n26.XY Walls Height
    Contrast=ScalarParameter() {defaultValue=8, parameterName=EdgeBlend_Contrast}
   ) {materialFunction=&HeightLerp}.Alpha
  )
  n48 = VirtualTextureFeatureSwitch(
   No=@n46.Result
   Yes=StaticSwitchParameter(True=@n46.Result, False=StaticSwitchParameter(True=@n47, False=@n46.Result) {parameterName=UseRockEdgeBlend}) {parameterName=UseCustomMaterialBlend}
  )
  n49 = QualitySwitch(Default=@n48, Low=@n46.Result, High=@n48, Medium=@n46.Result)
  n50 = StaticSwitchParameter(
   True=BlendMaterialAttributes(A=@n49, B=@n24, Alpha=StaticSwitchParameter(True=@n23, False=@n22) {desc=Diffuse, group=PaintLayer, parameterName=Modulate Paint Layer Mask})
   False=@n49
  ) {desc=Diffuse, group=Layers, parameterName=Use PaintLayer}
  n51 = VirtualTextureFeatureSwitch(No=@n50, Yes=BlendMaterialAttributes(A=@n6, B=@n50, Alpha=Saturate(None=@n13)) {vertexAttributeBlendType=UseA})
  n52 = GetMaterialAttributes(None=QualitySwitch(Default=@n51, Low=@n50, High=@n51, Medium=@n50)) {attributeGetTypes [
    #1
    #2
    #3
    #4
   ]}
  n53 = SetMaterialAttributes!(MaterialAttributes, Base Color=@n52.Base Color, Specular=@n52.Base Color, Roughness=@n52.Base Color, Normal=@n52.Base Color) {attributeSetTypes [
    #1
    #2
    #3
    #4
   ]}
  SetMaterialAttributes!(
   MaterialAttributes=MaterialFunctionCall(MA In=@n53) {materialFunction=&MF_CreativeMoveTool}.MA Out
   Pixel Depth Offset=Constant()
  ) {attributeSetTypes [#5]}
  n54 = TextureObjectParameter() {
   group=Grass
   parameterName=GrassDiffuse
   samplerSource=SSM_Wrap_WorldGroupSettings
   texture=/Game/Packages/DS_Fortnite_Terrain_NoLOD/Textures/T_Grasslands_D
  }
  n55 = TextureObjectParameter() {
   group=Grass
   parameterName=GrassNormal
   samplerSource=SSM_Wrap_WorldGroupSettings
   samplerType=SAMPLERTYPE_Normal
   texture=/Game/Packages/DS_Fortnite_Terrain_NoLOD/Textures/T_Grasslands_N
  }
  n56 = MaterialFunctionCall!(
   Diffuse Texture Object=@n54
   Spec Texture Object
   TextureSize=ScalarParameter() {defaultValue=700, group=Grass, parameterName=GrassTiling}
   --------------------------
   Use seperate settings for Z projection
   Use seperate texture for Z projection
   Texture Size Z projection
   Z Plane Diffuse Texture Object
   Z Plane Spec Texture Object
   "--------------------------      "
   ProjectionTransitionContrast
   Use Texture Alpha Heightmap to Refine the Transistion
   HeightLerpTransitionContrast
   PickTheDiffuseHeightChannel
   TheDiffuseHeightChannel
   "--------------------------       "
   NormalTextureObject=@n55
   World Space Normals
   "--------------------------  "
   Export Float 4 Diffuse=StaticBool() {value=true}
   Export Float 4 Spec
   "--------------------------     "
   WorldPosition
   World Space Normal
  ) {materialFunction=/Engine/Functions/Engine_MaterialFunctions02/Texturing/WorldAlignedTextures_Complex}
  StaticSwitchParameter(True=@n56.Normal XYZ Texture, False=@n56.Normal XYZ Texture) {group=Grass, parameterName=trueGrassTriplanar}
  n57 = StaticSwitchParameter(True=@n56.XYZ Diffuse Texture, False=@n56.XYZ Diffuse Texture) {group=Grass, parameterName=trueGrassTriplanar}
  n58 = Constant()
  n59 = ComponentMask(None=@n57) {a=true}
  n60 = ScalarParameter() {defaultValue=2, group=Grass, parameterName=GrassSpecPower}
  n61 = OneMinus(None=@n59)
  n62 = MakeMaterialAttributes(
   BaseColor=@n57
   Metallic=@n58
   Specular=LinearInterpolate(B=ScalarParameter() {defaultValue=0.3, group=Grass, parameterName=GrassSpecAmount}, Alpha=Power(Base=@n59, Exp=@n60))
   Roughness=ScalarParameter() {defaultValue=0.8, group=Grass, parameterName=GrassRoughness}
   Opacity=@n61
  )
  n63 = TextureObjectParameter() {
   group=BaseTilingTexture
   parameterName=Base Tiling Texture
   samplerSource=SSM_Wrap_WorldGroupSettings
   texture=/Game/Packages/DS_Fortnite_Terrain_NoLOD/Textures/T_FORT_Underground_02
  }
  n64 = TextureObjectParameter() {
   group=BaseTilingTexture
   parameterName=Base Tiling Texture_SpecMask
   samplerSource=SSM_Wrap_WorldGroupSettings
   texture=&T_BlacksRGB
  }
  n65 = TextureObjectParameter() {
   group=BaseTilingTexture
   parameterName=Base Tiling Normal Texture
   samplerSource=SSM_Wrap_WorldGroupSettings
   samplerType=SAMPLERTYPE_Normal
   texture=/Clyde/ExtraDependencies/T_Normal
  }
  n66 = MaterialFunctionCall(
   Diffuse Texture Object=@n63
   Spec Texture Object=@n64
   TextureSize=ScalarParameter() {defaultValue=-512, group=BaseTilingTexture, parameterName=Base Tiling Texture Tile Size}
   NormalTextureObject=@n65
   Export Float 4 Diffuse=StaticBool() {value=true}
  ) {materialFunction=&WorldAlignedTextures_Complex}
  n67 = MaterialFunctionCall!(In=StaticSwitchParameter(True=@n66.XYZ Mat Attributes - See Tooltip, False=@n66.XYZ Mat Attributes - See Tooltip) {group=BaseTilingTexture, parameterName="Make Base Texture project XYZ(true) or Z(False)"}) {materialFunction=/Game/Packages/Fortress_SharedMaterials/Base_Material_Functions/SwizzleWorldAlignedSpecOutputForFortnite}
  n68 = StaticBoolParameter() {group=BaseTilingTexture, parameterName=Use Base Texture Spec}
  n69 = StaticBoolParameter() {defaultValue=true, group=BaseTilingTexture, parameterName=Use Base Texture Normals}
  n70 = Constant()
  StaticSwitchParameter(
   True=StaticSwitchParameter(
    True=@n62
    False=MaterialFunctionCall!(Mat attributes=@n67.Result, Use Spec=@n68, Use Normals=@n69) {materialFunction=/Game/Packages/Fortnite_UniqueMaterials_1/MaterialFunctions/Fortnite_Terrain_invertOpacityNormalBool}.Result
   ) {defaultValue=true, group=Layers, parameterName="Use Grass(true) or a BaseTilingTexture(False)"}
   False=MakeMaterialAttributes(BaseColor=@n70, Specular=@n70, Opacity=Constant() {r=1})
  ) {defaultValue=true, group=Layers, parameterName=Use a base Layer}
  StaticSwitchParameter(True=StaticSwitchParameter(True=Constant() {r=1}, False=Constant()) {defaultValue=true, group=Layers, parameterName="Use Grass(true) or a BaseTilingTexture(False)"}) {defaultValue=true, group=Layers, parameterName=Use a base Layer}
  n71 = RuntimeVirtualTextureSample!(UVs, World Position, Mip Level) {
   materialType=BaseColor_Normal_Specular_YCoCg
   mipValueMode=RVTMVM_MipLevel
   virtualTexture=/Game/Athena/Apollo/Maps/Landscape/VT/RVT_Apollo_Landscape
  }
  VirtualTextureFeatureSwitch(
   No=@n62
   Yes=MakeMaterialAttributes(BaseColor=@n71.BaseColor, Metallic=@n58, Specular=@n71.BaseColor, Roughness=@n71.BaseColor, Opacity=@n61, Normal=@n71.BaseColor)
  )
 ]
}
